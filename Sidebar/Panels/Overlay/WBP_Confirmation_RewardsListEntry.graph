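# Reward row widget for a confirmation list, showing a reward item's icon, name, stack count and special-reward marker.
# SetReward stores the item and type, then UpdateReward sets visibility and switches on item type to common, people, schematic or transform-key layouts.
# A warning icon appears for gold-containing card packs when the owning player's gold would exceed its stack limit.
# Add to Reward Count increments the displayed count for merged entries.
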WidgetBlueprint /Sidebar/Panels/Overlay/WBP_Confirmation_RewardsListEntry
graphs [
 Add to Reward Count {nodes [
   n1 = Add to Reward Count!() -> (then, Value:double) {is=title K2Node_FunctionEntry}
   n2 = float + float(
    A=@n1
    B=Get CurrentNumericValue!(self:Common Numeric Text Block Object Reference=Get RewardCount() -> (RewardCount:Common Numeric Text Block Object Reference)) -> (CurrentNumericValue:float)
   )
   > SetCurrentValue!(execute, self:Common Numeric Text Block Object Reference=Get RewardCount(), NewValue:float=@n2) -> (then)
   > SetVisibility(self=Get RewardCount(), InVisibility=HitTestInvisible)
   n3 = "Equal (Enum)"(
    A:EFortItemType Enum=GetType!(self:Fort Item Object Reference=Get RewardItem() -> (RewardItem:Fort Item Object Reference)) -> (ReturnValue:EFortItemType Enum=WorldItem)
    B:EFortItemType Enum=CardPack
   )
   n4 = "float > float"(A=@n2, B=1)
   n5 = Select!(
    NewEnumerator0:bool=true
    NewEnumerator1:bool=false
    NewEnumerator2:bool=false
    Index:Confirmation_RewardListEntryType Enum=Get RewardType() -> (RewardType:Confirmation_RewardListEntryType Enum=NewEnumerator0)
   ) -> (ReturnValue:bool)
   n6 = Get ShouldShowLabel() -> (ShouldShowLabel:bool)
   > SetVisibility(
    self=Get TextCardPackCount() -> (TextCardPackCount:Common Text Object Reference)
    InVisibility=Select!(
     Option 0:ESlateVisibility Enum=Collapsed
     Option 1:ESlateVisibility Enum=SelfHitTestInvisible
     Index:bool=AND Boolean!(A=@n3, B=@n4, C:bool=@n6, D:bool=@n5) -> (ReturnValue)
    ) -> (ReturnValue:ESlateVisibility Enum)
   )
  ]}
 EventGraph {nodes [
   ConfigForSchematic() {is=title K2Node_CustomEvent}
   > SetActiveWidgetIndex(self=Get SwitcherRewardType() -> (SwitcherRewardType:Common Widget Switcher Legacy Object Reference), Index=1)
   n1 = Get RewardItem()
   > SetItemToRepresent!(
    execute
    self:Fort Item Widget NUI Object Reference=Get FortItemCard_XXSSchematicItemCard() -> (FortItemCard_XXSSchematicItemCard:Fort Multi Size Item Card Object Reference)
    ItemToRepresent:Fort Item Object Reference=@n1
   ) -> (then)
   > SetColorAndOpacity(
    self=Get RewardName() -> (RewardName:Common Text Object Reference)
    InColorAndOpacity=Make Slate Color(SpecifiedColor=GetRarityColor!(self:Fort Item Object Reference=@n1) -> (ReturnValue:Linear Color Structure))
   )
   ConfigForPeople() {is=title K2Node_CustomEvent}
   > SetActiveWidgetIndex(self=Get SwitcherRewardType(), Index=0)
   > SetBrushFromLazyTexture!(
    execute
    self:Common Lazy Image Object Reference=Get LazyImageRewardIcon() -> (LazyImageRewardIcon:Common Lazy Image Object Reference)
    LazyTexture:"Texture 2D Soft Object Reference (by ref)"=GetSmallPreviewImage!(self:Fort Item Object Reference=Get RewardItem()) -> (ReturnValue:Texture 2D Soft Object Reference)
    bMatchSize:bool=false
   ) -> (then)
   n2 = Get RewardItem()
   > SetColorAndOpacity(self=Get RewardName(), InColorAndOpacity=Make Slate Color(SpecifiedColor=GetRarityColor(self=@n2)))
   n3 = Get RewardItem()
   > n4 = Branch(Condition="Equal Exactly (Text)"!(
     A:"Text (by ref)"=GetDisplayName!(self:Fort Item Object Reference=@n3) -> (ReturnValue:Text)
     B:"Text (by ref)"=GetEmptyText!() -> (ReturnValue:Text)
    ) -> (ReturnValue:bool))
   > "SetText (Text)"(
    self=Get RewardName()
    InText=GetPrimaryCategory!(self:Fort Item Object Reference=Get RewardItem()) -> (ItemCategoryText:Text, ReturnValue:bool)
   )
   n5 = Get RewardItem()
   n6 = Get RewardItem()
   "SetText (Text)"(
    execute=@n4.else
    self=Get RewardName()
    InText=Format Text!(Format="LOCTEXT(\"{a} ({b})\")", a:Text=GetDisplayName(self=@n5), b:Text=GetPrimaryCategory(self=@n6)) -> (Result)
   )
   ConfigForTransformKey() {is=title K2Node_CustomEvent}
   > SetActiveWidgetIndex(self=Get SwitcherRewardType(), Index=0)
   > SetBrushFromLazyTexture(self=Get LazyImageRewardIcon(), LazyTexture=GetSmallPreviewImage(self=Get RewardItem()), bMatchSize=false)
   n7 = Get RewardItem()
   > "SetText (Text)"(
    self=Get RewardName()
    InText=Format Text!(Format="LOCTEXT(\"Transform Key: {key name}\")", key name:Text=GetDisplayName(self=@n7)) -> (Result)
   )
   n8 = Get RewardItem()
   > SetColorAndOpacity(self=Get RewardName(), InColorAndOpacity=Make Slate Color(SpecifiedColor=GetRarityColor(self=@n8)))
   ConfigForCommonRewards() {is=title K2Node_CustomEvent}
   > SetActiveWidgetIndex(self=Get SwitcherRewardType(), Index=0)
   > SetBrushFromLazyTexture(self=Get LazyImageRewardIcon(), LazyTexture=GetSmallPreviewImage(self=Get RewardItem()), bMatchSize=false)
   UpdateReward!(execute, self, InTextColor:Slate Color Structure=Get TextColor() -> (TextColor:Slate Color Structure)) -> (then)
   Event Construct()
  ]}
 GetRewardName {nodes [
   GetRewardName() {is=title K2Node_FunctionEntry}
   > Return Node!(execute, RewardName:Text="GetText (Text)"!(self:Text Object Reference=Get RewardName()) -> (ReturnValue:Text))
  ]}
 SetMaxDesiredWidth {nodes [
   n1 = SetMaxDesiredWidth!() -> (then, InMaxDesiredWidth:float=~) {is=title K2Node_FunctionEntry}
   > SetMaxDesiredWidth!(
    execute
    self:Size Box Object Reference=Get MaxDesiredWidthBox() -> (MaxDesiredWidthBox:Size Box Object Reference)
    InMaxDesiredWidth:float=@n1
   ) -> (then) {is=title K2Node_CallFunction}
  ]}
 SetReward {nodes [
   n1 = SetReward!() -> (
    then
    InRewardType:Confirmation_RewardListEntryType Enum=NewEnumerator0
    InRewardItem:Fort Item Object Reference
    InTextColor:Slate Color Structure
   ) {is=title K2Node_FunctionEntry}
   > Branch(Condition="Not Equal (Object)"!(A:Object Reference=@n1.InRewardItem, B:Object Reference=Get RewardItem()) -> (ReturnValue:bool))
   > Set RewardItem(RewardItem:Fort Item Object Reference=@n1.InRewardItem) -> (Output_Get:Fort Item Object Reference)
   > Set RewardType(RewardType:Confirmation_RewardListEntryType Enum=@n1) -> (Output_Get:Confirmation_RewardListEntryType Enum=NewEnumerator0)
   > UpdateReward(InTextColor=Get InTextColor() -> (InTextColor:Slate Color Structure))
  ]}
 UpdateReward {nodes [
   UpdateReward!() -> (then, InTextColor:Slate Color Structure) {is=title K2Node_FunctionEntry}
   > n1 = Is Valid(InputObject=Get RewardItem())
   > SetVisibility(
    self=Get SwitcherSpecialRewardType() -> (SwitcherSpecialRewardType:Common Widget Switcher Legacy Object Reference)
    InVisibility=Select!(
     NewEnumerator0:ESlateVisibility Enum=Collapsed
     NewEnumerator1:ESlateVisibility Enum=SelfHitTestInvisible
     NewEnumerator2:ESlateVisibility Enum=SelfHitTestInvisible
     Index:Confirmation_RewardListEntryType Enum=Get RewardType() -> (RewardType=NewEnumerator0)
    ) -> (ReturnValue)
   )
   > SetActiveWidgetIndex(
    self=Get SwitcherSpecialRewardType()
    Index=Select(NewEnumerator0:int=0, NewEnumerator1:int=0, NewEnumerator2:int=1, Index=Get RewardType() -> (RewardType=NewEnumerator0)) -> (ReturnValue:int)
   )
   > "SetText (Text)"(self=Get RewardName(), InText=GetDisplayName(self=Get RewardItem()))
   > SetVisibility(
    self=Get RewardName()
    InVisibility=Select!(Option 0:ESlateVisibility Enum=Collapsed, Option 1:ESlateVisibility Enum=SelfHitTestInvisible, Index:bool=Get ShouldShowLabel()) -> (ReturnValue:ESlateVisibility Enum)
   )
   n2 = GetNumInStack!(self:Fort Item Object Reference=Get RewardItem()) -> (ReturnValue:int=0)
   n3 = "integer > integer"(A=@n2, B=1) {is=title K2Node_CallFunction}
   n4 = Get ShouldShowLabel()
   > SetVisibility(
    self=Get RewardCount()
    InVisibility=Select(Option 0=Collapsed, Option 1=HitTestInvisible, Index=AND Boolean!(A=@n3, B=@n4) -> (ReturnValue))
   )
   > SetCurrentValue(self=Get RewardCount(), NewValue="To Float (Integer)"(InInt=@n2))
   > SetColorAndOpacity(self=[Get RewardName(), Get RewardCount()], InColorAndOpacity=Get InTextColor())
   n5 = Get RewardItem()
   n6 = "integer > integer"(A=GetNumInStack(self=@n5) -> (ReturnValue=0), B=1)
   n7 = "Equal (Enum)"(A=GetType(self=@n5) -> (ReturnValue=WorldItem), B=CardPack)
   n8 = Select!(
    NewEnumerator0:bool=true
    NewEnumerator1:bool=false
    NewEnumerator2:bool=false
    Index:Confirmation_RewardListEntryType Enum=Get RewardType() -> (RewardType=NewEnumerator0)
   ) -> (ReturnValue:bool)
   n9 = Get ShouldShowLabel()
   > SetVisibility(
    self=Get TextCardPackCount()
    InVisibility=Select!(
     Option 0:ESlateVisibility Enum=Collapsed
     Option 1:ESlateVisibility Enum=SelfHitTestInvisible
     Index:bool=AND Boolean!(A=@n7, B=@n6, C:bool=@n9, D:bool=@n8) -> (ReturnValue)
    ) -> (ReturnValue:ESlateVisibility Enum)
   )
   > SetVisibility(self=Get ImageMaxLimitWarning() -> (ImageMaxLimitWarning:Image Object Reference), InVisibility=Collapsed)
   > n10 = Cast To FortInventoryOwnerInterface!(execute, Object=GetOwningPlayer()) -> (then, CastFailed, AsFort Inventory Owner Interface:Fort Inventory Owner Interface Interface) {is=title K2Node_DynamicCast}
   > n11 = GetInventoryItemWithTag!(
    execute
    InventoryOwner:Fort Inventory Owner Interface Interface=@n10
    TagToFind:"Gameplay Tag Structure (by ref)"=MakeLiteralGameplayTag!(Value:Gameplay Tag Structure=(TagName="Item.Resource.Gold")) -> (ReturnValue:Gameplay Tag Structure)
    bExactMatch:bool=false
   ) -> (then, ReturnValue:Fort Item Object Reference)
   > n12 = Is Valid(InputObject=@n11)
   n13 = WouldExceedMaxStackSize!(
    self:Fort Inventory Context Object Reference=Get FortInventoryContext!() -> (ReturnValue:Fort Inventory Context Object Reference) {is=title K2Node_GetBlueprintContextEx}
    ItemDefinition:Fort Item Definition Object Reference=GetItemDefinition!(self:Fort Item Object Reference=@n11) -> (ReturnValue:Fort Item Definition Object Reference)
    QuantityToTest:int=1
   ) -> (ReturnValue:bool)
   n14 = Get RewardItem()
   n15 = HasTag!(
    TagContainer:"Gameplay Tag Container Structure (by ref)"=GetOwnedGameplayTags!(TagContainerInterface:Gameplay Tag Asset Interface Interface=GetItemDefinition(self=@n14)) -> (ReturnValue:Gameplay Tag Container Structure)
    Tag:Gameplay Tag Structure=(TagName="MissionReward.CardPack.ContainsGold")
    bExactMatch:bool=true
   ) -> (ReturnValue:bool)
   > n16 = SetVisibility(
    self=Get ImageMaxLimitWarning()
    InVisibility=Select(Option 0=Collapsed, Option 1=SelfHitTestInvisible, Index=AND Boolean!(A=@n15, B=@n13) -> (ReturnValue))
   )
   n17 = Switch on EFortItemType!(execute=[@n16.then, @n12.Is Not Valid], Selection:EFortItemType Enum=GetType(self=Get RewardItem()) -> (ReturnValue=WorldItem)) -> (
    WorldItem:Exec
    Ammo:Exec
    Badge:Exec
    BackpackPickup:Exec
    BuildingPiece:Exec
    CharacterPart:Exec
    Consumable:Exec
    Deco:Exec
    EditTool:Exec
    Ingredient:Exec
    ItemCache:Exec
    Food:Exec
    Gadget:Exec
    AthenaGadget:Exec
    HomebaseGadget:Exec
    SpyTechPerk:Exec
    HeroAbility:Exec
    MissionItem:Exec
    Trap:Exec
    MultiItem:Exec
    Weapon:Exec
    WeaponMelee:Exec
    WeaponRanged:Exec
    WeaponHarvest:Exec
    WeaponCreativePhone:Exec
    WeaponMod:Exec
    WorldResource:Exec
    CreativeUserPrefab:Exec
    CreativePlayset:Exec
    Vehicle:Exec
    Npc:Exec
    PlayerAugment:Exec
    AccountItem:Exec
    AccountResource:Exec
    CollectedResource:Exec
    Alteration:Exec
    CardPack:Exec
    Currency:Exec
    Hero:Exec
    Schematic:Exec
    Worker:Exec
    TeamPerk:Exec
    PlayerTech:Exec
    Token:Exec
    DailyRewardScheduleToken:Exec
    CodeToken:Exec
    Stat:Exec
    Buff:Exec
    BuffCredit:Exec
    Quest:Exec
    Accolades:Exec
    FriendChest:Exec
    MedalsPunchCard:Exec
    RepeatableDailiesCard:Exec
    ChallengeBundle:Exec
    ChallengeBundleSchedule:Exec
    ChallengeBundleCompletionToken:Exec
    GameplayModifier:Exec
    Outpost:Exec
    HomebaseNode:Exec
    Defender:Exec
    ConversionControl:Exec
    DeployableBaseCloudSave:Exec
    ConsumableAccountItem:Exec
    Quota:Exec
    Expedition:Exec
    ExpeditionV2:Exec
    HomebaseBannerIcon:Exec
    HomebaseBannerColor:Exec
    AthenaSkyDiveContrail:Exec
    PersonalVehicle:Exec
    AthenaGlider:Exec
    AthenaPickaxe:Exec
    AthenaBackpack:Exec
    CosmeticShoes:Exec
    CosmeticMimosa:Exec
    CosmeticMimosaA:Exec
    CosmeticMimosaB:Exec
    CosmeticMimosaC:Exec
    AthenaCharacter:Exec
    AthenaDance:Exec
    AthenaLoadingScreen:Exec
    AthenaBattleBus:Exec
    AthenaVehicleCosmetic:Exec
    AthenaItemWrap:Exec
    AthenaMusicPack:Exec
    AthenaPetCosmetic:Exec
    AthenaSeasonTreasure:Exec
    AthenaSeason:Exec
    AthenaRewardGraph:Exec
    AthenaExtResource:Exec
    EventDescription:Exec
    AthenaEventToken:Exec
    EventPurchaseTracker:Exec
    CosmeticVariantToken:Exec
    CampaignHeroLoadout:Exec
    Playset:Exec
    PrerollData:Exec
    CreativePlot:Exec
    PlayerSurveyToken:Exec
    CosmeticLocker:Exec
    BannerToken:Exec
    RestedXpBoosterToken:Exec
    RewardEventGraphPurchaseToken:Exec
    EventDependentItem:Exec
    ItemAccessToken:Exec
    STWAccoladeReward:Exec
    Campsite:Exec
    VictoryCrown:Exec
    RealitySapling:Exec
    Apparel:Exec
    ApparelLayout:Exec
    PlayerAugmentsPersistence:Exec
    SparksAura:Exec
    SparksGuitar:Exec
    SparksBass:Exec
    SparksKeyboard:Exec
    SparksMicrophone:Exec
    SparksDrums:Exec
    SparksSpotlightAnim:Exec
    SparksSong:Exec
    JunoBuildingSet:Exec
    JunoBuildingProp:Exec
    MagpieEntitlementReward:Exec
    DeferredItemGrant:Exec
    XpBooster:Exec
    SpecialItem:Exec
    Emote:Exec
    Stack:Exec
    CollectionBookPage:Exec
    BGAConsumableWrapper:Exec
    GiftBox:Exec
    GiftBoxUnlock:Exec
    PlaysetProp:Exec
    RegCosmeticDef:Exec
    Profile:Exec
    Max_None:Exec
   ) {is=title K2Node_SwitchEnum}
   ConfigForCommonRewards(execute=[@n17.AccountResource, @n17.Badge, @n17.CardPack, @n17.Currency, @n17.Ingredient, @n17.Token, @n17.WorldItem, @n17.WorldResource]) {is=title K2Node_CallFunction}
   ConfigForTransformKey(execute=@n17.ConversionControl) {is=title K2Node_CallFunction}
   ConfigForSchematic(execute=@n17.Schematic) {is=title K2Node_CallFunction}
   ConfigForPeople(execute=[@n17.Defender, @n17.Hero, @n17.Worker]) {is=title K2Node_CallFunction}
   SetVisibility(execute=@n1.Is Not Valid, self=Get SwitcherSpecialRewardType(), InVisibility=Collapsed)
   > SetVisibility(self=Get RewardName(), InVisibility=Collapsed)
   > SetVisibility(self=Get RewardCount(), InVisibility=Collapsed)
   > SetVisibility(self=Get TextCardPackCount(), InVisibility=Collapsed)
  ]}
]
variables [
 RewardItem {type=FortItem, default=null}
 RewardType {type=Confirmation_RewardListEntryType, default=NormalReward}
 ShouldShowLabel {type=boolean, default=true}
 TextColor {type=SlateColor, default {colorUseRule=UseColor_Specified, specifiedColor=<rgba 1 1 1 1>}}
]
widgets {
 info {inheritedWidgetCount=0, parentClass=&CommonUserWidget, rootWidgetClass=&SizeBox, widgetCount=21}
 widgets [
  MaxDesiredWidthBox {widgetClassPath=&SizeBox, bIsVariable=true, widget=MaxDesiredWidthBox}
  HorizontalBox_0 {widgetClassPath=&HorizontalBox, parent=MaxDesiredWidthBox, widget=HorizontalBox_0, slot {class=SizeBoxSlot}}
  SwitcherSpecialRewardType {
   widgetClassPath=&CommonWidgetSwitcherLegacy
   bIsVariable=true
   parent=HorizontalBox_0
   widget=SwitcherSpecialRewardType
   properties {transitionDuration=0}
   slot {class=HorizontalBoxSlot, properties {horizontalAlignment=HAlign_Center, verticalAlignment=VAlign_Center}}
  }
  SizeBoxMissionAlert {
   widgetClassPath=&SizeBox
   parent=SwitcherSpecialRewardType
   widget=SizeBoxMissionAlert
   properties {bOverride_WidthOverride=true, widthOverride=32, renderTransform {angle=0, scale=<1 1>, shear=<0 0>, translation=<4 -2>}}
   slot {class=WidgetSwitcherSlot, properties {horizontalAlignment=HAlign_Center, verticalAlignment=VAlign_Center}}
  }
  Overlay_3 {widgetClassPath=&Overlay, parent=SizeBoxMissionAlert, widget=Overlay_3, slot {class=SizeBoxSlot}}
  CommonBorder_4 {
   widgetClassPath=&CommonBorder
   parent=Overlay_3
   widget=CommonBorder_4
   properties {
    horizontalAlignment=HAlign_Center
    visibility=Collapsed
    background {
     bIsDynamicallyLoaded=false
     drawAs=Box
     imageType=NoImage
     mirroring=NoMirror
     resourceName=None
     resourceObject=/FortUILibrary/Styles/Border/Materials/M-Border-Bang
     tiling=NoTile
     imageSize=<32 32>
     margin=<ltrb 0.5962 0 0.6991 0>
     outlineSettings {
      bUseBrushTransparency=false
      roundingType=HalfHeightRadius
      width=0
      color {colorUseRule=UseColor_Specified, specifiedColor=<rgba 0 0 0 0>}
      cornerRadii=<0 0 0 0>
     }
     tintColor {colorUseRule=UseColor_Specified, specifiedColor=<rgba 1 1 1 1>}
     uVRegion {bIsValid=false, max=<0 0>, min=<0 0>}
    }
    brushColor=<rgba 0 0 0 0.6>
    padding=<ltrb 0 0 0 0>
    renderTransform {angle=0, scale=<1 1>, shear=<0 0>, translation=<1 1>}
   }
   slot {class=OverlaySlot, properties {horizontalAlignment=HAlign_Fill, verticalAlignment=VAlign_Fill}}
  }
  Border_Small {
   widgetClassPath=&CommonBorder
   parent=Overlay_3
   widget=Border_Small
   properties {
    horizontalAlignment=HAlign_Center
    visibility=Collapsed
    background {
     bIsDynamicallyLoaded=false
     drawAs=Box
     imageType=NoImage
     mirroring=NoMirror
     resourceName=None
     resourceObject=&M-Border-Bang
     tiling=NoTile
     imageSize=<32 32>
     margin=<ltrb 0.5962 0 0.6991 0>
     outlineSettings {
      bUseBrushTransparency=false
      roundingType=HalfHeightRadius
      width=0
      color {colorUseRule=UseColor_Specified, specifiedColor=<rgba 0 0 0 0>}
      cornerRadii=<0 0 0 0>
     }
     tintColor {colorUseRule=UseColor_Specified, specifiedColor=<rgba 1 1 1 1>}
     uVRegion {bIsValid=false, max=<0 0>, min=<0 0>}
    }
    padding=<ltrb 0 0 0 0>
   }
   slot {class=OverlaySlot, properties {horizontalAlignment=HAlign_Fill, verticalAlignment=VAlign_Fill}}
  }
  SizeBox_3 {
   widgetClassPath=&SizeBox
   parent=Overlay_3
   widget=SizeBox_3
   properties {bOverride_HeightOverride=true, bOverride_WidthOverride=true, heightOverride=24, visibility=Collapsed, widthOverride=24}
   slot {class=OverlaySlot, properties {horizontalAlignment=HAlign_Center, verticalAlignment=VAlign_Fill}}
  }
  CommonTextBlock_5 {
   widgetClassPath=&CommonTextBlock
   bIsVariable=true
   parent=SizeBox_3
   widget=CommonTextBlock_5
   properties {
    justification=Center
    style=/FortUILibrary/Styles/Typography/Foundation/TextStyle-Base-XS-B-Black.TextStyle-Base-XS-B-Black_C
    text="!"
    colorAndOpacity {colorUseRule=UseColor_Specified, specifiedColor=<rgba 0 0 0 1>}
    font {
     bForceMonospaced=false
     bMaterialIsStencil=false
     fontMaterial=null
     fontObject=&NotoSans
     letterSpacing=0
     monospacedWidth=1
     size=16
     skewAmount=0
     typefaceFontName=NotoSans-Bold
     outlineSettings {
      bApplyOutlineToDropShadows=false
      bMiteredCorners=false
      bSeparateFillAlpha=false
      outlineMaterial=null
      outlineSize=0
      outlineColor=<rgba 0 0 0 1>
     }
    }
    renderTransform {angle=0, scale=<1 1>, shear=<0 0>, translation=<0 -4>}
    shadowOffset=<0 0>
   }
   slot {class=SizeBoxSlot}
  }
  Image_4 {
   widgetClassPath=&Image
   bIsVariable=true
   parent=Overlay_3
   widget=Image_4
   properties {
    brush {
     bIsDynamicallyLoaded=false
     drawAs=Image
     imageType=NoImage
     mirroring=NoMirror
     resourceName=None
     resourceObject=/Game/UI/Foundation/Textures/Icons/HUD/T-Icon-Clock-128
     tiling=NoTile
     imageSize=<24 24>
     margin=<ltrb 0 0 0 0>
     outlineSettings {
      bUseBrushTransparency=false
      roundingType=HalfHeightRadius
      width=0
      color {colorUseRule=UseColor_Specified, specifiedColor=<rgba 0 0 0 0>}
      cornerRadii=<0 0 0 0>
     }
     tintColor {colorUseRule=UseColor_Specified, specifiedColor=<rgba 0 0 0 1>}
     uVRegion {bIsValid=false, max=<0 0>, min=<0 0>}
    }
    renderTransform {angle=0, scale=<1 1>, shear=<0 0>, translation=<1 1>}
   }
   slot {class=OverlaySlot, properties {horizontalAlignment=HAlign_Center, verticalAlignment=VAlign_Center}}
  }
  Image_3 {
   widgetClassPath=&Image
   bIsVariable=true
   parent=Overlay_3
   widget=Image_3
   properties {brush {
     bIsDynamicallyLoaded=false
     drawAs=Image
     imageType=NoImage
     mirroring=NoMirror
     resourceName=None
     resourceObject=&T-Icon-Clock-128
     tiling=NoTile
     imageSize=<24 24>
     margin=<ltrb 0 0 0 0>
     outlineSettings {
      bUseBrushTransparency=false
      roundingType=HalfHeightRadius
      width=0
      color {colorUseRule=UseColor_Specified, specifiedColor=<rgba 0 0 0 0>}
      cornerRadii=<0 0 0 0>
     }
     tintColor {colorUseRule=UseColor_Specified, specifiedColor=<rgba 1 1 1 1>}
     uVRegion {bIsValid=false, max=<0 0>, min=<0 0>}
    }}
   slot {class=OverlaySlot, properties {horizontalAlignment=HAlign_Center, verticalAlignment=VAlign_Center}}
  }
  SizeBox_2 {
   widgetClassPath=&SizeBox
   parent=SwitcherSpecialRewardType
   widget=SizeBox_2
   properties {bOverride_WidthOverride=true, widthOverride=32}
   slot {class=WidgetSwitcherSlot}
  }
  ImageFoundersReward {widgetClassPath=&Image, bIsVariable=true, parent=SizeBox_2, widget=ImageFoundersReward, slot {class=SizeBoxSlot}}
  Spacer_0 {widgetClassPath=&Spacer, parent=HorizontalBox_0, widget=Spacer_0, properties {size=<8 1>}, slot {class=HorizontalBoxSlot}}
  SwitcherRewardType {
   widgetClassPath=&CommonWidgetSwitcherLegacy
   bIsVariable=true
   parent=HorizontalBox_0
   widget=SwitcherRewardType
   properties {transitionDuration=0}
   slot {class=HorizontalBoxSlot, properties {horizontalAlignment=HAlign_Center, verticalAlignment=VAlign_Center}}
  }
  LazyImageRewardIcon {
   widgetClassPath=/Script/CommonUI.CommonLazyImage
   bIsVariable=true
   parent=SwitcherRewardType
   widget=LazyImageRewardIcon
   properties {
    visibility=SelfHitTestInvisible
    brush {
     bIsDynamicallyLoaded=false
     drawAs=Image
     imageType=NoImage
     mirroring=NoMirror
     resourceName=None
     resourceObject=/Game/UI/Foundation/Textures/Icons/ItemTypes/T-Icon-Explosive-64
     tiling=NoTile
     imageSize=<40 40>
     margin=<ltrb 0 0 0 0>
     outlineSettings {
      bUseBrushTransparency=false
      roundingType=HalfHeightRadius
      width=0
      color {colorUseRule=UseColor_Specified, specifiedColor=<rgba 0 0 0 0>}
      cornerRadii=<0 0 0 0>
     }
     tintColor {colorUseRule=UseColor_Specified, specifiedColor=<rgba 1 1 1 1>}
     uVRegion {bIsValid=false, max=<0 0>, min=<0 0>}
    }
   }
   slot {class=WidgetSwitcherSlot, properties {horizontalAlignment=HAlign_Left, verticalAlignment=VAlign_Center}}
  }
  FortItemCard_XXSSchematicItemCard {
   widgetClassPath=/Script/FortniteUI.FortMultiSizeItemCard
   bIsVariable=true
   parent=SwitcherRewardType
   widget=FortItemCard_XXSSchematicItemCard
   properties {emptyItemType=AccountItem}
   slot {class=WidgetSwitcherSlot, properties {horizontalAlignment=HAlign_Center, verticalAlignment=VAlign_Center}}
  }
  ImageMaxLimitWarning {
   widgetClassPath=&Image
   bIsVariable=true
   parent=HorizontalBox_0
   widget=ImageMaxLimitWarning
   properties {
    visibility=Collapsed
    brush {
     bIsDynamicallyLoaded=false
     drawAs=Image
     imageType=NoImage
     mirroring=NoMirror
     resourceName=None
     resourceObject=/Game/UI/Foundation/Shell/Textures/T-Icon-Warning-32
     tiling=NoTile
     imageSize=<32 32>
     margin=<ltrb 0 0 0 0>
     outlineSettings {
      bUseBrushTransparency=false
      roundingType=HalfHeightRadius
      width=0
      color {colorUseRule=UseColor_Specified, specifiedColor=<rgba 0 0 0 0>}
      cornerRadii=<0 0 0 0>
     }
     tintColor {colorUseRule=UseColor_Specified, specifiedColor=<rgba 1 1 1 1>}
     uVRegion {bIsValid=false, max=<0 0>, min=<0 0>}
    }
    colorAndOpacity=<rgba 1 0.9047 0.0648 1>
   }
   slot {class=HorizontalBoxSlot, properties {horizontalAlignment=HAlign_Center, verticalAlignment=VAlign_Center}}
  }
  RewardCount {
   widgetClassPath=/Script/CommonUI.CommonNumericTextBlock
   bIsVariable=true
   parent=HorizontalBox_0
   widget=RewardCount
   properties {
    style=/FortUILibrary/Styles/Typography/Foundation/TextStyle-Base-XS-B.TextStyle-Base-XS-B_C
    text=0
    font {
     bForceMonospaced=false
     bMaterialIsStencil=false
     fontMaterial=null
     fontObject=&NotoSans
     letterSpacing=0
     monospacedWidth=1
     size=16
     skewAmount=0
     typefaceFontName=NotoSans-Bold
     outlineSettings {
      bApplyOutlineToDropShadows=false
      bMiteredCorners=false
      bSeparateFillAlpha=false
      outlineMaterial=null
      outlineSize=0
      outlineColor=<rgba 0 0 0 1>
     }
    }
    shadowOffset=<0 0>
   }
   slot {class=HorizontalBoxSlot, properties {horizontalAlignment=HAlign_Center, verticalAlignment=VAlign_Center, padding=<ltrb 8 0 8 0>}}
  }
  TextCardPackCount {
   widgetClassPath=&CommonTextBlock
   bIsVariable=true
   parent=HorizontalBox_0
   widget=TextCardPackCount
   properties {
    style=&TextStyle-Base-XS-B_C
    text=x
    visibility=Collapsed
    font {
     bForceMonospaced=false
     bMaterialIsStencil=false
     fontMaterial=null
     fontObject=&NotoSans
     letterSpacing=0
     monospacedWidth=1
     size=16
     skewAmount=0
     typefaceFontName=NotoSans-Bold
     outlineSettings {
      bApplyOutlineToDropShadows=false
      bMiteredCorners=false
      bSeparateFillAlpha=false
      outlineMaterial=null
      outlineSize=0
      outlineColor=<rgba 0 0 0 1>
     }
    }
    shadowOffset=<0 0>
   }
   slot {class=HorizontalBoxSlot, properties {verticalAlignment=VAlign_Center, padding=<ltrb -8 0 4 0>}}
  }
  RewardName {
   widgetClassPath=&CommonTextBlock
   bIsVariable=true
   parent=HorizontalBox_0
   widget=RewardName
   properties {
    scrollStyle=&TextScrollStyle-Base_C
    style=&TextStyle-Base-XS-B_C
    text=Reward Name
    font {
     bForceMonospaced=false
     bMaterialIsStencil=false
     fontMaterial=null
     fontObject=&NotoSans
     letterSpacing=0
     monospacedWidth=1
     size=16
     skewAmount=0
     typefaceFontName=NotoSans-Bold
     outlineSettings {
      bApplyOutlineToDropShadows=false
      bMiteredCorners=false
      bSeparateFillAlpha=false
      outlineMaterial=null
      outlineSize=0
      outlineColor=<rgba 0 0 0 1>
     }
    }
    shadowOffset=<0 0>
   }
   slot {
    class=HorizontalBoxSlot
    properties {horizontalAlignment=HAlign_Left, verticalAlignment=VAlign_Center, padding=<ltrb 8 0 0 0>, size {sizeRule=Fill, value=1}}
   }
  }
 ]
}
