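# Confirmation-panel widget that shows a mission's rewards in a vertical WrapBox of pre-placed reward entry widgets.
# Update orders mission alert rewards first, fills entries, folds same-named consecutive rewards into the previous entry's count and collapses unused ones.
# It also hides gold-containing card pack rewards when the player's gold stack is full.
# OnPaint infers the column count from child positions and splits the container width between columns as each entry's max width.
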WidgetBlueprint /Sidebar/Panels/Overlay/WBP_Confirmation_RewardsListEntryContainer
graphs [
 Compare With Previous Reward {nodes [
   n1 = Compare With Previous Reward!() -> (then, ItemToCompare:Fort Item Object Reference, PreviousSlot:WBP Confirmation Rewards List Entry Object Reference) {is=title K2Node_FunctionEntry}
   > n2 = Branch(Condition=AND Boolean(A=IsValid(Object=@n1), B=IsValid(Object=@n1.PreviousSlot)))
   n3 = Get ItemToCompare() -> (ItemToCompare:Fort Item Object Reference)
   n4 = Get RewardName!(self:WBP Confirmation Rewards List Entry Object Reference=Get PreviousSlot() -> (PreviousSlot:WBP Confirmation Rewards List Entry Object Reference)) -> (RewardName:Common Text Object Reference)
   > n5 = Branch(Condition="Equal Exactly (Text)"!(
     A:"Text (by ref)"="GetText (Text)"!(self:Text Object Reference=@n4) -> (ReturnValue:Text)
     B:"Text (by ref)"=GetDisplayName!(self:Fort Item Object Reference=@n3) -> (ReturnValue:Text)
    ) -> (ReturnValue:bool))
   n6 = Get ItemToCompare()
   > Add to Reward Count!(
    execute
    self:WBP Confirmation Rewards List Entry Object Reference=Get PreviousSlot()
    Value:double="To Float (Integer)"(InInt=GetNumInStack!(self:Fort Item Object Reference=@n6) -> (ReturnValue:int=0))
   ) -> (then)
   > Return Node!(execute, IsRepeated:bool=true)
   Return Node(execute=@n5.else, IsRepeated=false)
   Return Node(execute=@n2.else, IsRepeated=false)
  ]}
 EventGraph {links [], nodes []}
 Get Reward Items {nodes [
   Get Reward Items() {is=title K2Node_FunctionEntry}
   > Return Node!(
    execute
    Item Definitions:Array of Fort Item Definition Object References=Get Reward Item Definitions() -> (Reward Item Definitions:Array of Fort Item Definition Object References)
   )
  ]}
 GetWrapColumnsNum {nodes [
   GetWrapColumnsNum() {is=title K2Node_FunctionEntry}
   > Set LastLocalPosition(LastLocalPosition:double=0) -> (Output_Get:double)
   > Set Output(Output:int=1) -> (Output_Get:int=0)
   n1 = Get RewardListEntryContainer() -> (RewardListEntryContainer:Wrap Box Object Reference)
   > n2 = For Loop(LastIndex=int - int(A=GetChildrenCount!(self:Panel Widget Object Reference=@n1) -> (ReturnValue:int=0), B=1) {is=title K2Node_CallFunction})
   Return Node!(execute=@n2.Completed, ReturnValue:int=Get Output() -> (Output:int=0))
   n3 = Get RewardListEntryContainer()
   n4 = GetCachedGeometry!(self:Widget Object Reference=GetChildAt!(self:Panel Widget Object Reference=@n3, Index:int=@n2) -> (ReturnValue:Widget Object Reference)) -> (ReturnValue:"Geometry Structure (by ref)")
   n5 = BreakVector2D(InVec="Absolute to Local (Coordinates)"!(
     Geometry:"Geometry Structure (by ref)"=GetCachedGeometry(self=@n3)
     AbsoluteCoordinate:Vector 2D Structure=GetAbsolutePosition!(Geometry:"Geometry Structure (by ref)"=@n4) -> (ReturnValue:Vector 2D Structure)
    ) -> (ReturnValue:Vector 2D Structure))
   > Branch(Condition="float > float"(A=@n5, B=Get LastLocalPosition() -> (LastLocalPosition:double)))
   > Set LastLocalPosition(LastLocalPosition=@n5)
   > Increment Int(Value=Get Output() -> (Output=0))
  ]}
 OnPaint {nodes [
   OnPaint!() -> (then, Context:"Paint Context Structure (by ref)") {is=title K2Node_FunctionEntry}
   > n1 = GetWrapColumnsNum!(execute, self) -> (then, ReturnValue:int=0) {is=title K2Node_CallFunction}
   > Set Columns(Columns:int=@n1) -> (Output_Get:int=0)
   n2 = GetLocalSize!(Geometry:"Geometry Structure (by ref)"=GetCachedGeometry(self=Get RewardListEntryContainer())) -> (ReturnValue:Vector 2D Structure)
   > n3 = Branch(Condition="float > float"(A=BreakVector2D(InVec=@n2)))
   n4 = Get RewardListEntryContainer()
   > n5 = For Loop(LastIndex=int - int(A=GetChildrenCount(self=@n4) -> (ReturnValue=0), B=1))
   Return Node!(execute=[@n5.Completed, @n3.else])
   > n6 = Cast To WBP_Confirmation_RewardsListEntry!(execute, Object=GetChildAt(self=Get RewardListEntryContainer(), Index=@n5)) -> (then, CastFailed, AsWBP Confirmation Rewards List Entry:WBP Confirmation Rewards List Entry Object Reference) {is=title K2Node_DynamicCast}
   n7 = GetLocalSize(Geometry=GetCachedGeometry(self=Get RewardListEntryContainer()))
   > SetMaxDesiredWidth!(
    execute
    self:WBP Confirmation Rewards List Entry Object Reference=@n6
    InMaxDesiredWidth:float=float / float(A=BreakVector2D(InVec=@n7), B:int=Get Columns() -> (Columns:int=0))
   ) -> (then)
  ]}
 OverMaxDisplayLimit {nodes [
   OverMaxDisplayLimit!() -> (then, Modifiers:Array of Fort Gameplay Modifier Item Definition Object References) {is=title K2Node_FunctionEntry}
   n1 = Length(TargetArray:Array of Fort Mission Reward Info Structures=Get Mission Rewards Array() -> (Mission Rewards Array:Array of Fort Mission Reward Info Structures))
   n2 = Length(TargetArray:Array of Fort Gameplay Modifier Item Definition Object References=Get Modifiers() -> (Modifiers:Array of Fort Gameplay Modifier Item Definition Object References))
   > n3 = Branch(Condition=AND Boolean(A="integer >= integer"!(A:int=@n1, B:int=5) -> (ReturnValue:bool) {is=title K2Node_CallFunction}, B="integer >= integer"(A=@n2, B=2)))
   > Return Node!(execute, Overlimit:bool=true)
   Return Node(execute=@n3.else, Overlimit=false)
  ]}
 Update {nodes [
   Update!() -> (
    then
    Mission Rewards:Array of Fort Mission Reward Info Structures
    Gameplay Modifiers:Array of Fort Gameplay Modifier Item Definition Object References
   ) {is=title K2Node_FunctionEntry}
   > Clear(TargetArray:Array of Fort Mission Reward Info Structures=Get Mission Rewards Array())
   > Clear(TargetArray:Array of Fort Item Definition Object References=Get Reward Item Definitions())
   > n1 = For Each Loop(Array:Array of Fort Mission Reward Info Structures=Get Mission Rewards() -> (Mission Rewards:Array of Fort Mission Reward Info Structures)) -> (Array Element:Fort Mission Reward Info Structure)
   n2 = For Each Loop(Exec=@n1.Completed, Array=Get Mission Rewards())
   n3 = Break Fort Mission Reward Info!(FortMissionRewardInfo:"Fort Mission Reward Info Structure (by ref)"=@n2) -> (Item:Fort Item Object Reference, bIsMissionAlertReward:bool=~) {is=title K2Node_BreakStruct}
   > n4 = Branch(Condition=@n3.bIsMissionAlertReward)
   Add(
    execute=@n4.else
    TargetArray:Array of Fort Item Definition Object References=Get Reward Item Definitions()
    NewItem:"Fort Item Definition Object Reference (by ref)"=GetItemDefinition!(self:Fort Item Object Reference=@n3) -> (ReturnValue:Fort Item Definition Object Reference)
   ) {is=title K2Node_CallArrayFunction}
   > Add(
    TargetArray:Array of Fort Mission Reward Info Structures=Get Mission Rewards Array()
    NewItem:"Fort Mission Reward Info Structure (by ref)"=@n2
   )
   Set CurrRewardIndex(execute=@n2.Completed, CurrRewardIndex:int=0) -> (Output_Get:int=0)
   n5 = Get RewardListEntryContainer()
   > n6 = For Loop(LastIndex=int - int(A=GetChildrenCount(self=@n5) -> (ReturnValue=0), B=1))
   > n7 = Cast To WBP_Confirmation_RewardsListEntry(Object=GetChildAt(self=Get RewardListEntryContainer(), Index=@n6))
   > Set CurrSlot(CurrSlot:WBP Confirmation Rewards List Entry Object Reference=@n7) -> (Output_Get:WBP Confirmation Rewards List Entry Object Reference)
   n8 = Get Mission Rewards Array()
   > n9 = Branch(Condition="integer < integer"!(A:int=Get CurrRewardIndex() -> (CurrRewardIndex:int=0), B:int=Length(TargetArray:Array of Fort Mission Reward Info Structures=@n8)) -> (ReturnValue:bool) {is=title K2Node_CallFunction})
   n10 = Break Fort Mission Reward Info!(FortMissionRewardInfo="Get (a ref)"(Array:Array of Fort Mission Reward Info Structures=Get Mission Rewards Array(), Dimension 1=Get CurrRewardIndex() -> (CurrRewardIndex=0)) -> (Output:"Fort Mission Reward Info Structure (by ref)")) -> (DisplayName:Text, Item, bIsMissionAlertReward=~)
   > Set CurrRewardItem(CurrRewardItem:Fort Item Object Reference=@n10.Item) -> (Output_Get:Fort Item Object Reference)
   > Set CurrRewardIsMissionAlert(CurrRewardIsMissionAlert:bool=@n10.bIsMissionAlertReward) -> (Output_Get:bool)
   > n11 = Compare With Previous Reward!(
    execute
    self
    ItemToCompare:Fort Item Object Reference=Get CurrRewardItem() -> (CurrRewardItem:Fort Item Object Reference)
    PreviousSlot:WBP Confirmation Rewards List Entry Object Reference=Get PrevSlot() -> (PrevSlot:WBP Confirmation Rewards List Entry Object Reference)
   ) -> (then, IsRepeated:bool) {is=title K2Node_CallFunction}
   > n12 = Branch(Condition=@n11)
   SetVisibility(
    execute=[@n19.then, @n12.then]
    self=Get CurrSlot() -> (CurrSlot:WBP Confirmation Rewards List Entry Object Reference)
    InVisibility=Collapsed
   )
   > Increment Int(Value=Get CurrRewardIndex() -> (CurrRewardIndex=0))
   n13 = Cast To FortInventoryOwnerInterface!(execute=@n12.else, Object=GetOwningPlayer()) -> (then, CastFailed, AsFort Inventory Owner Interface:Fort Inventory Owner Interface Interface) {is=title K2Node_DynamicCast}
   > n14 = GetInventoryItemWithTag!(
    execute
    InventoryOwner:Fort Inventory Owner Interface Interface=@n13
    TagToFind:"Gameplay Tag Structure (by ref)"=MakeLiteralGameplayTag!(Value:Gameplay Tag Structure=(TagName="Item.Resource.Gold")) -> (ReturnValue:Gameplay Tag Structure)
    bExactMatch:bool=false
   ) -> (then, ReturnValue:Fort Item Object Reference)
   > n15 = Is Valid(InputObject=@n14)
   n16 = GetOwnedGameplayTags!(TagContainerInterface:Gameplay Tag Asset Interface Interface=GetItemDefinition(self=Get CurrRewardItem())) -> (ReturnValue:Gameplay Tag Container Structure)
   n17 = GetItemDefinition(self=@n14)
   n18 = Get FortInventoryContext!() -> (ReturnValue:Fort Inventory Context Object Reference) {is=title K2Node_GetBlueprintContextEx}
   > n19 = Branch(Condition=AND Boolean(
     A=HasTag!(
      TagContainer:"Gameplay Tag Container Structure (by ref)"=@n16
      Tag:Gameplay Tag Structure=(TagName="MissionReward.CardPack.ContainsGold")
      bExactMatch:bool=true
     ) -> (ReturnValue:bool)
     B=WouldExceedMaxStackSize!(self:Fort Inventory Context Object Reference=@n18, ItemDefinition:Fort Item Definition Object Reference=@n17, QuantityToTest:int=1) -> (ReturnValue:bool)
    ))
   SetReward!(
    execute=[@n19.else, @n15.Is Not Valid, @n13.CastFailed]
    self:WBP Confirmation Rewards List Entry Object Reference=Get CurrSlot()
    InRewardType:Confirmation_RewardListEntryType Enum=Select(
     Option 0:Confirmation_RewardListEntryType Enum=NewEnumerator0
     Option 1:Confirmation_RewardListEntryType Enum=NewEnumerator1
     Index=Get CurrRewardIsMissionAlert() -> (CurrRewardIsMissionAlert:bool)
    ) -> (ReturnValue:Confirmation_RewardListEntryType Enum)
    InRewardItem:Fort Item Object Reference=Get CurrRewardItem()
    InTextColor:Slate Color Structure=Make Slate Color(SpecifiedColor=(R=1,G=1,B=1,A=1))
   ) -> (then)
   > SetVisibility(self=Get CurrSlot(), InVisibility=SelfHitTestInvisible)
   > Set PrevSlot(PrevSlot:WBP Confirmation Rewards List Entry Object Reference=Get CurrSlot()) -> (Output_Get:WBP Confirmation Rewards List Entry Object Reference)
   > Increment Int(Value=Get CurrRewardIndex() -> (CurrRewardIndex=0))
   SetVisibility(execute=@n9.else, self=Get CurrSlot(), InVisibility=Collapsed)
   n20 = Break Fort Mission Reward Info!(FortMissionRewardInfo=@n1) -> (Item, bIsMissionAlertReward=~)
   Branch(execute=@n1.LoopBody, Condition=@n20.bIsMissionAlertReward)
   > Add(
    TargetArray:Array of Fort Item Definition Object References=Get Reward Item Definitions()
    NewItem:"Fort Item Definition Object Reference (by ref)"=GetItemDefinition(self=@n20)
   )
   > Add(
    TargetArray:Array of Fort Mission Reward Info Structures=Get Mission Rewards Array()
    NewItem:"Fort Mission Reward Info Structure (by ref)"=@n1
   )
  ]}
]
variables [
 Mission Rewards Array {type="array<FortMissionRewardInfo>", default []}
 Reward Item Definitions {type="array<FortItemDefinition>", default []}
]
widgets {
 info {inheritedWidgetCount=0, parentClass=&CommonUserWidget, rootWidgetClass=&WrapBox, widgetCount=17}
 widgets [
  RewardListEntryContainer {
   widgetClassPath=&WrapBox
   bIsVariable=true
   widget=RewardListEntryContainer
   properties {bExplicitWrapSize=true, bOverride_Cursor=true, orientation=Orient_Vertical, wrapSize=310}
  }
  WBP_Confirmation_RewardsListEntry_0 {
   widgetClassPath=&/WBP_Confirmation_RewardsListEntry
   parent=RewardListEntryContainer
   widget=WBP_Confirmation_RewardsListEntry_0
   slot {class=WrapBoxSlot}
  }
  WBP_Confirmation_RewardsListEntry_1 {
   widgetClassPath=&WBP_Confirmation_RewardsListEntry
   bIsVariable=true
   parent=RewardListEntryContainer
   widget=WBP_Confirmation_RewardsListEntry_1
   slot {class=WrapBoxSlot}
  }
  WBP_Confirmation_RewardsListEntry_2 {
   widgetClassPath=&WBP_Confirmation_RewardsListEntry
   bIsVariable=true
   parent=RewardListEntryContainer
   widget=WBP_Confirmation_RewardsListEntry_2
   slot {class=WrapBoxSlot}
  }
  WBP_Confirmation_RewardsListEntry_3 {
   widgetClassPath=&WBP_Confirmation_RewardsListEntry
   bIsVariable=true
   parent=RewardListEntryContainer
   widget=WBP_Confirmation_RewardsListEntry_3
   slot {class=WrapBoxSlot}
  }
  WBP_Confirmation_RewardsListEntry_4 {
   widgetClassPath=&WBP_Confirmation_RewardsListEntry
   bIsVariable=true
   parent=RewardListEntryContainer
   widget=WBP_Confirmation_RewardsListEntry_4
   slot {class=WrapBoxSlot}
  }
  WBP_Confirmation_RewardsListEntry_5 {
   widgetClassPath=&WBP_Confirmation_RewardsListEntry
   bIsVariable=true
   parent=RewardListEntryContainer
   widget=WBP_Confirmation_RewardsListEntry_5
   slot {class=WrapBoxSlot}
  }
  WBP_Confirmation_RewardsListEntry_6 {
   widgetClassPath=&WBP_Confirmation_RewardsListEntry
   bIsVariable=true
   parent=RewardListEntryContainer
   widget=WBP_Confirmation_RewardsListEntry_6
   slot {class=WrapBoxSlot}
  }
  WBP_Confirmation_RewardsListEntry_7 {
   widgetClassPath=&WBP_Confirmation_RewardsListEntry
   bIsVariable=true
   parent=RewardListEntryContainer
   widget=WBP_Confirmation_RewardsListEntry_7
   slot {class=WrapBoxSlot}
  }
  WBP_Confirmation_RewardsListEntry_8 {
   widgetClassPath=&WBP_Confirmation_RewardsListEntry
   bIsVariable=true
   parent=RewardListEntryContainer
   widget=WBP_Confirmation_RewardsListEntry_8
   slot {class=WrapBoxSlot}
  }
  WBP_Confirmation_RewardsListEntry_9 {
   widgetClassPath=&WBP_Confirmation_RewardsListEntry
   bIsVariable=true
   parent=RewardListEntryContainer
   widget=WBP_Confirmation_RewardsListEntry_9
   slot {class=WrapBoxSlot}
  }
  WBP_Confirmation_RewardsListEntry_10 {
   widgetClassPath=&WBP_Confirmation_RewardsListEntry
   bIsVariable=true
   parent=RewardListEntryContainer
   widget=WBP_Confirmation_RewardsListEntry_10
   slot {class=WrapBoxSlot}
  }
  WBP_Confirmation_RewardsListEntry_11 {
   widgetClassPath=&WBP_Confirmation_RewardsListEntry
   bIsVariable=true
   parent=RewardListEntryContainer
   widget=WBP_Confirmation_RewardsListEntry_11
   slot {class=WrapBoxSlot}
  }
  WBP_Confirmation_RewardsListEntry_12 {
   widgetClassPath=&WBP_Confirmation_RewardsListEntry
   bIsVariable=true
   parent=RewardListEntryContainer
   widget=WBP_Confirmation_RewardsListEntry_12
   slot {class=WrapBoxSlot}
  }
  WBP_Confirmation_RewardsListEntry_13 {
   widgetClassPath=&WBP_Confirmation_RewardsListEntry
   bIsVariable=true
   parent=RewardListEntryContainer
   widget=WBP_Confirmation_RewardsListEntry_13
   slot {class=WrapBoxSlot}
  }
  WBP_Confirmation_RewardsListEntry_14 {
   widgetClassPath=&WBP_Confirmation_RewardsListEntry
   bIsVariable=true
   parent=RewardListEntryContainer
   widget=WBP_Confirmation_RewardsListEntry_14
   slot {class=WrapBoxSlot}
  }
  WBP_Confirmation_RewardsListEntry_15 {
   widgetClassPath=&WBP_Confirmation_RewardsListEntry
   bIsVariable=true
   parent=RewardListEntryContainer
   widget=WBP_Confirmation_RewardsListEntry_15
   slot {class=WrapBoxSlot}
  }
 ]
}
